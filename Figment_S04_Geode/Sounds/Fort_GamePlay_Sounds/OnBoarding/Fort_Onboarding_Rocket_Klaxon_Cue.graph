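# Sound cue for the onboarding rocket klaxon in the Figment_S04_Geode gameplay sounds.
# A single looping wave player is the root and plays the Fort_Onboarding_Rocket_Klaxon wave with no mixing or randomization.
# The cue routes through the SFX sound class at reduced volume.
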SoundCue /Figment_S04_Geode/Sounds/Fort_GamePlay_Sounds/OnBoarding/Fort_Onboarding_Rocket_Klaxon_Cue
root=WavePlayer_11
nodes [{
  id=WavePlayer_11
  properties {bLooping=true, soundWaveAssetPtr=&/Fort_Onboarding_Rocket_Klaxon}
 }]
properties {maxDistance=2097152, soundClassObject=&SFX, volumeMultiplier=0.5}
