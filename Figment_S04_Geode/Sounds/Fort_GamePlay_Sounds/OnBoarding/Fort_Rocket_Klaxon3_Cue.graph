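# Sound cue for a rocket klaxon in the Fort onboarding gameplay sounds of Figment_S04_Geode.
# A single Wave Player feeds the Fort_Onboarding_Rocket_Klaxon wave into a Looping node set to a finite loop count of three instead of looping indefinitely.
# The cue plays at reduced volume in the SFX sound class.
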SoundCue /Figment_S04_Geode/Sounds/Fort_GamePlay_Sounds/OnBoarding/Fort_Rocket_Klaxon3_Cue
root=Looping_0
nodes [
 {id=Looping_0, children [WavePlayer_11], properties {bLoopIndefinitely=false, loopCount=3}}
 {id=WavePlayer_11, properties {soundWaveAssetPtr=&/Fort_Onboarding_Rocket_Klaxon}}
]
properties {maxDistance=2097152, soundClassObject=&SFX, volumeMultiplier=0.5}
